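# Physics asset for the SonnetSpirit gadget prop mesh, whose bone names indicate a three-segment snowman.
# The bottom, middle and top snowman joints each carry a capsule body, shrinking from bottom to top, while a tiny capsule sits on the main bone.
# Each segment is constrained directly to main rather than chained to its neighbor, with matching swing and twist limit angles.
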PhysicsAsset /BRCosmetics/Animation/Props/SonnetSpirit/Meshes/SonnetSpirit_Gadget_PhysicsAsset
bodies {
 JNT_Snowman_Bot [{shapeType=Capsule, length=33.0954, radius=33.0631, center=<2.3801 0 22.1693>, rotation=<rot 0 -90 -134.5951>}]
 JNT_Snowman_Mid [{shapeType=Capsule, length=24.9122, radius=24.3642, center=<-0.891 0.9181 17.1468>, rotation=<rot 1.973 -90 -110.5449>}]
 JNT_Snowman_Top [{shapeType=Capsule, length=16.9693, radius=16.6735, center=<0.0272 1.547 10.0041>, rotation=<rot -0.6395 -90 -126.997>}]
 main [{shapeType=Capsule, length=0.505, radius=0.505, rotation=<rot 0 -90 0>}]
}
constraints [
 {bone1Name=JNT_Snowman_Bot, bone2Name=main, swing1LimitDegrees=45, swing2LimitDegrees=45, twistLimitDegrees=45}
 {bone1Name=JNT_Snowman_Mid, bone2Name=main, swing1LimitDegrees=45, swing2LimitDegrees=45, twistLimitDegrees=45}
 {bone1Name=JNT_Snowman_Top, bone2Name=main, swing1LimitDegrees=45, swing2LimitDegrees=45, twistLimitDegrees=45}
]
